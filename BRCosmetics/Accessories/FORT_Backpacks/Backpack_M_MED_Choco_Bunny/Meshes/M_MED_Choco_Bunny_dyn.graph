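# Physics asset for the Choco Bunny backpack cosmetic, defining collision bodies and joints for its dynamic flower and strap bones.
# A single large capsule on the pack bone anchors three chains of small capsules: two flower bodies, a four-segment left strap and a five-segment right strap.
# Each link is constrained to its parent with limited swing and tight twist, with the strap roots held stiffer than the other joints.
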PhysicsAsset /BRCosmetics/Accessories/FORT_Backpacks/Backpack_M_MED_Choco_Bunny/Meshes/M_MED_Choco_Bunny_dyn
bodies {
 dyn_flowers_01 [{shapeType=Capsule, length=1, radius=5, center=<-1.5 0 0>, rotation=<rot 90 0 0>}]
 dyn_flowers_02 [{shapeType=Capsule, length=1, radius=5, center=<-1.5 0 0>, rotation=<rot 90 0 0>}]
 dyn_straps_l_01 [{shapeType=Capsule, length=1, radius=1, rotation=<rot 90 0 0>}]
 dyn_straps_l_02 [{shapeType=Capsule, length=1.3611, radius=1.5022, center=<2.3786 -1.4488 0.5206>, rotation=<rot 90 0 -89.9995>}]
 dyn_straps_l_03 [{shapeType=Capsule, length=0.4978, radius=1.5022, center=<0.9741 -1.3394 -0.0002>, rotation=<rot 90 90 -4.9972>}]
 dyn_straps_l_04 [{shapeType=Capsule, length=1.2017, radius=1.3423, center=<1.3333 0 0>, rotation=<rot 90 0 0>}]
 dyn_straps_r_01 [{shapeType=Capsule, length=1, radius=1, rotation=<rot 90 0 0>}]
 dyn_straps_r_02 [{shapeType=Capsule, length=1, radius=1, rotation=<rot 90 0 0>}]
 dyn_straps_r_03 [{shapeType=Capsule, length=1, radius=1.1828, center=<0.6703 0.9888 0.5724>, rotation=<rot 69.9956 -84.9971 179.9999>}]
 dyn_straps_r_04 [{shapeType=Capsule, length=1, radius=1.1828, center=<0.178 1.3374 0.6219>, rotation=<rot 69.9976 -89.9979 -179.9998>}]
 dyn_straps_r_05 [{shapeType=Capsule, length=1.2017, radius=1.2423, center=<1.513 0 0>, rotation=<rot 90 0 0>}]
 pack [{shapeType=Capsule, length=6.3564, radius=11.2003, center=<0.0264 -10.8537 0.3918>}]
}
constraints [
 {bone1Name=dyn_flowers_01, bone2Name=pack, swing1LimitDegrees=45, swing2LimitDegrees=45, twistLimitDegrees=5}
 {bone1Name=dyn_flowers_02, bone2Name=dyn_flowers_01, swing1LimitDegrees=45, swing2LimitDegrees=45, twistLimitDegrees=5}
 {bone1Name=dyn_straps_l_01, bone2Name=pack, swing1LimitDegrees=30, swing2LimitDegrees=30, twistLimitDegrees=5}
 {bone1Name=dyn_straps_l_02, bone2Name=dyn_straps_l_01, swing1LimitDegrees=45, swing2LimitDegrees=45, twistLimitDegrees=5}
 {bone1Name=dyn_straps_l_03, bone2Name=dyn_straps_l_02, swing1LimitDegrees=45, swing2LimitDegrees=45, twistLimitDegrees=5}
 {bone1Name=dyn_straps_l_04, bone2Name=dyn_straps_l_03, swing1LimitDegrees=45, swing2LimitDegrees=45, twistLimitDegrees=5}
 {bone1Name=dyn_straps_r_01, bone2Name=pack, swing1LimitDegrees=20, swing2LimitDegrees=20, twistLimitDegrees=5}
 {bone1Name=dyn_straps_r_02, bone2Name=dyn_straps_r_01, swing1LimitDegrees=45, swing2LimitDegrees=45, twistLimitDegrees=5}
 {bone1Name=dyn_straps_r_03, bone2Name=dyn_straps_r_02, swing1LimitDegrees=45, swing2LimitDegrees=45, twistLimitDegrees=5}
 {bone1Name=dyn_straps_r_04, bone2Name=dyn_straps_r_03, swing1LimitDegrees=45, swing2LimitDegrees=45, twistLimitDegrees=5}
 {bone1Name=dyn_straps_r_05, bone2Name=dyn_straps_r_04, swing1LimitDegrees=45, swing2LimitDegrees=45, twistLimitDegrees=5}
]
